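# Physics asset for the F_MED_Razor female character body that defines secondary-motion bodies for its necklace, bell, headphones and bow.
# Capsule bodies on the pelvis and upper spine form the base, linked to each other by free-rotation constraints.
# The necklace, which carries the bell, and the headphones attach to spine_05 with angular limits.
# The bow's left and right parts branch from a center body, linked through a tiny master body that rotates freely on the pelvis.
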PhysicsAsset /BRCosmetics/Characters/Player/Female/Medium/Bodies/F_MED_Razor/Meshes/F_MED_Razor_Dyn
bodies {
 dyn_bell [{shapeType=Capsule, length=0, radius=3.5, center=<3.8741 0.2713 0.0001>, rotation=<rot 90 0 0>}]
 dyn_bow [{shapeType=Capsule, length=2, radius=1, center=<0.545 0 0>, rotation=<rot 90 0 0>}]
 dyn_bow_MSTR [{shapeType=Capsule, length=0.1, radius=0.1}]
 dyn_bow_l [{shapeType=Capsule, length=2, radius=1, center=<0.1 0 0>, rotation=<rot 90 0 0>}]
 dyn_bow_r [{shapeType=Capsule, length=2, radius=1, center=<0.1 0 0>, rotation=<rot 90 0 0>}]
 dyn_headphones [{shapeType=Capsule, length=2, radius=1, center=<-0.1 0 0>, rotation=<rot 90 0 0>}]
 dyn_necklace [{shapeType=Capsule, length=2.5, radius=2.5, shapeName=dyn_necklace_capsule, center=<2.3677 0 0>}]
 pelvis [{shapeType=Capsule, length=4.75, radius=14.25, center=<-4.8623 0.8233 -0.0001>}]
 spine_04 [{shapeType=Capsule, length=11.5, radius=7, center=<-0.3276 7.922 0>}]
 spine_05 [{shapeType=Capsule, length=10.25, radius=9.75, center=<-2.1872 -1.1181 0>}]
}
constraints [
 {bone1Name=dyn_bell, bone2Name=dyn_necklace, swing1LimitDegrees=45, swing2LimitDegrees=45, twistLimitDegrees=5}
 {
  bone1Name=dyn_bow_MSTR
  bone2Name=pelvis
  swing1LimitDegrees=45
  swing1Motion=Free
  swing2LimitDegrees=45
  swing2Motion=Free
  twistLimitDegrees=45
  twistMotion=Free
 }
 {bone1Name=dyn_bow, bone2Name=dyn_bow_MSTR, swing1LimitDegrees=45, swing2LimitDegrees=45, twistLimitDegrees=5}
 {bone1Name=dyn_bow_r, bone2Name=dyn_bow, swing1LimitDegrees=45, swing2LimitDegrees=45, twistLimitDegrees=5}
 {bone1Name=dyn_bow_l, bone2Name=dyn_bow, swing1LimitDegrees=45, swing2LimitDegrees=45, twistLimitDegrees=5}
 {bone1Name=dyn_necklace, bone2Name=spine_05, swing1LimitDegrees=45, swing2LimitDegrees=20, twistLimitDegrees=15}
 {bone1Name=dyn_headphones, bone2Name=spine_05, swing1LimitDegrees=45, swing2LimitDegrees=45, twistLimitDegrees=5}
 {
  bone1Name=spine_05
  bone2Name=spine_04
  swing1LimitDegrees=45
  swing1Motion=Free
  swing2LimitDegrees=45
  swing2Motion=Free
  twistLimitDegrees=45
  twistMotion=Free
 }
 {
  bone1Name=spine_04
  bone2Name=pelvis
  swing1LimitDegrees=45
  swing1Motion=Free
  swing2LimitDegrees=45
  swing2Motion=Free
  twistLimitDegrees=45
  twistMotion=Free
 }
]
