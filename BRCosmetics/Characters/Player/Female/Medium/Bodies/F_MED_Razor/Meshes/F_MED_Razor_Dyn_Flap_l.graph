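# Physics asset for secondary motion of the left-side flaps, pendant and feathers on the F_MED_Razor player cosmetic body.
# Five three-body flap chains (A–E) join the pelvis with free rotation, then bend mostly on one axis through tight swing and twist limits.
# The pendant swings on one axis from flap A's middle body and carries a four-capsule feather chain.
# A left thigh capsule, freely constrained to the pelvis, provides a leg collision shape.
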PhysicsAsset /BRCosmetics/Characters/Player/Female/Medium/Bodies/F_MED_Razor/Meshes/F_MED_Razor_Dyn_Flap_l
bodies {
 dyn_feather_01 [{shapeType=Capsule, length=3.625, radius=1.2188, center=<2.756 0 0>, rotation=<rot 90 0 0>}]
 dyn_feather_02 [{shapeType=Capsule, length=7.9062, radius=2.5, center=<5.925 0 -0.0001>, rotation=<rot 90 0 0>}]
 dyn_feather_03 [{shapeType=Capsule, length=3, radius=3, center=<3.7643 0 0>, rotation=<rot 90 0 0>}]
 dyn_feather_04 [{shapeType=Capsule, length=1.5, radius=2.5, center=<2.5 0 0>, rotation=<rot 90 0 0>}]
 dyn_pendant [{shapeType=Sphere, radius=2, shapeName=dyn_pendant_sphere, center=<-0.1041 0 0.725>}]
 dyn_side_flap_A_01 [{shapeType=Box, extentX=7.5, extentY=7.5, extentZ=1.5, center=<3.3511 0 -0.0001>}]
 dyn_side_flap_A_02 [{shapeType=Capsule, length=7.5, radius=1.5, shapeName=dyn_side_flap_A_02_capsule, center=<3.3511 0 -0.0001>, rotation=<rot 90 0 0>}]
 dyn_side_flap_A_03 [{shapeType=Capsule, length=7.5, radius=1.5, shapeName=dyn_side_flap_A_03_capsule, center=<7.7 0 0>, rotation=<rot 90 0 0>}]
 dyn_side_flap_B_01 [{shapeType=Box, extentX=7.5, extentY=7.5, extentZ=1.5, center=<3.3511 0 -0.0001>}]
 dyn_side_flap_B_02 [{shapeType=Capsule, length=7, radius=1.5, shapeName=dyn_side_flap_B_02_capsule, center=<3.3511 0 -0.0001>, rotation=<rot 90 0 0>}]
 dyn_side_flap_B_03 [{shapeType=Capsule, length=7, radius=1.5, shapeName=dyn_side_flap_B_03_capsule, center=<7.7 0 0>, rotation=<rot 90 0 0>}]
 dyn_side_flap_C_01 [{shapeType=Box, extentX=7.5, extentY=7.5, extentZ=1.5, center=<3.3511 0 -0.0001>}]
 dyn_side_flap_C_02 [{shapeType=Capsule, length=7.5, radius=1.5, shapeName=dyn_side_flap_C_02_capsule, center=<3.3511 0 -0.0001>, rotation=<rot 90 0 0>}]
 dyn_side_flap_C_03 [{shapeType=Capsule, length=7.5, radius=1.5, shapeName=dyn_side_flap_C_03_capsule, center=<7.7 0 0>, rotation=<rot 90 0 0>}]
 dyn_side_flap_D_01 [{shapeType=Box, extentX=7.5, extentY=7.5, extentZ=1.5, center=<3.3511 0 -0.0001>}]
 dyn_side_flap_D_02 [{shapeType=Capsule, length=7.5, radius=1.5, shapeName=dyn_side_flap_D_02_capsule, center=<3.3511 0 -0.0001>, rotation=<rot 90 0 0>}]
 dyn_side_flap_D_03 [{shapeType=Capsule, length=7.5, radius=1.5, shapeName=dyn_side_flap_D_03_capsule, center=<7.7 0 0>, rotation=<rot 90 0 0>}]
 dyn_side_flap_E_01 [{shapeType=Box, extentX=7.5, extentY=7.5, extentZ=1.5, center=<3.3511 0 -0.0001>}]
 dyn_side_flap_E_02 [{shapeType=Capsule, length=7.5, radius=1.5, shapeName=dyn_side_flap_E_02_capsule, center=<3.3511 0 -0.0001>, rotation=<rot 90 0 0>}]
 dyn_side_flap_E_03 [{shapeType=Capsule, length=7.5, radius=1.5, shapeName=dyn_side_flap_E_03_capsule, center=<7.7 0 0>, rotation=<rot 90 0 0>}]
 pelvis [{shapeType=Capsule, length=1, radius=1}]
 thigh_l [{shapeType=Capsule, length=25.4555, radius=8.9464, center=<-20.7924 -0.0552 0.0001>, rotation=<rot 87.9578 0.9052 0>}]
}
constraints [
 {bone1Name=dyn_feather_01, bone2Name=dyn_pendant, swing1LimitDegrees=40, swing2LimitDegrees=40, twistLimitDegrees=5}
 {bone1Name=dyn_feather_02, bone2Name=dyn_feather_01, swing1LimitDegrees=80, swing2LimitDegrees=80, twistLimitDegrees=5}
 {bone1Name=dyn_feather_03, bone2Name=dyn_feather_02, swing1LimitDegrees=80, swing2LimitDegrees=80, twistLimitDegrees=5}
 {bone1Name=dyn_feather_04, bone2Name=dyn_feather_03, swing1LimitDegrees=80, swing2LimitDegrees=80, twistLimitDegrees=5}
 {
  bone1Name=dyn_pendant
  bone2Name=dyn_side_flap_A_02
  swing1LimitDegrees=90
  swing2LimitDegrees=45
  swing2Motion=Locked
  twistLimitDegrees=45
  twistMotion=Locked
 }
 {
  bone1Name=dyn_side_flap_A_01
  bone2Name=pelvis
  swing1LimitDegrees=45
  swing1Motion=Free
  swing2LimitDegrees=45
  swing2Motion=Free
  twistLimitDegrees=45
  twistMotion=Free
 }
 {bone1Name=dyn_side_flap_A_02, bone2Name=dyn_side_flap_A_01, swing1LimitDegrees=0.2, swing2LimitDegrees=10, twistLimitDegrees=0.5}
 {bone1Name=dyn_side_flap_A_03, bone2Name=dyn_side_flap_A_02, swing1LimitDegrees=0.2, swing2LimitDegrees=10, twistLimitDegrees=0.5}
 {
  bone1Name=dyn_side_flap_B_01
  bone2Name=pelvis
  swing1LimitDegrees=45
  swing1Motion=Free
  swing2LimitDegrees=45
  swing2Motion=Free
  twistLimitDegrees=45
  twistMotion=Free
 }
 {bone1Name=dyn_side_flap_B_02, bone2Name=dyn_side_flap_B_01, swing1LimitDegrees=0.2, swing2LimitDegrees=45, twistLimitDegrees=0.5}
 {bone1Name=dyn_side_flap_B_03, bone2Name=dyn_side_flap_B_02, swing1LimitDegrees=0.2, swing2LimitDegrees=45, twistLimitDegrees=0.5}
 {
  bone1Name=dyn_side_flap_C_01
  bone2Name=pelvis
  swing1LimitDegrees=45
  swing1Motion=Free
  swing2LimitDegrees=45
  swing2Motion=Free
  twistLimitDegrees=45
  twistMotion=Free
 }
 {bone1Name=dyn_side_flap_C_02, bone2Name=dyn_side_flap_C_01, swing1LimitDegrees=0.2, swing2LimitDegrees=45, twistLimitDegrees=0.5}
 {bone1Name=dyn_side_flap_C_03, bone2Name=dyn_side_flap_C_02, swing1LimitDegrees=0.2, swing2LimitDegrees=45, twistLimitDegrees=0.5}
 {
  bone1Name=dyn_side_flap_D_01
  bone2Name=pelvis
  swing1LimitDegrees=45
  swing1Motion=Free
  swing2LimitDegrees=45
  swing2Motion=Free
  twistLimitDegrees=45
  twistMotion=Free
 }
 {bone1Name=dyn_side_flap_D_02, bone2Name=dyn_side_flap_D_01, swing1LimitDegrees=0.2, swing2LimitDegrees=45, twistLimitDegrees=0.5}
 {bone1Name=dyn_side_flap_D_03, bone2Name=dyn_side_flap_D_02, swing1LimitDegrees=0.2, swing2LimitDegrees=45, twistLimitDegrees=0.5}
 {
  bone1Name=dyn_side_flap_E_01
  bone2Name=pelvis
  swing1LimitDegrees=45
  swing1Motion=Free
  swing2LimitDegrees=45
  swing2Motion=Free
  twistLimitDegrees=45
  twistMotion=Free
 }
 {bone1Name=dyn_side_flap_E_02, bone2Name=dyn_side_flap_E_01, swing1LimitDegrees=0.5, swing2LimitDegrees=45, twistLimitDegrees=0.5}
 {bone1Name=dyn_side_flap_E_03, bone2Name=dyn_side_flap_E_02, swing1LimitDegrees=0.5, swing2LimitDegrees=45, twistLimitDegrees=0.5}
 {
  bone1Name=thigh_l
  bone2Name=pelvis
  swing1LimitDegrees=45
  swing1Motion=Free
  swing2LimitDegrees=45
  swing2Motion=Free
  twistLimitDegrees=45
  twistMotion=Free
 }
]
