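# Physics asset for the dynamic right-foot tail bones on the medium female Razor player body.
# Small capsule bodies sit on the MSTR root bone and two tail segments.
# Two constraints link them as a chain from MSTR through 01 to 02, with moderate swing limits and tight twist limits, so the tail can bend but barely twist.
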PhysicsAsset /BRCosmetics/Characters/Player/Female/Medium/Bodies/F_MED_Razor/Meshes/F_MED_Razor_Dyn_Leg_r
bodies {
 dyn_foot_tail_r_01 [{shapeType=Capsule, length=2, radius=1, center=<-3 0 0>, rotation=<rot 90 0 0>}]
 dyn_foot_tail_r_02 [{shapeType=Capsule, length=2, radius=1, center=<-1 0 0>, rotation=<rot 90 0 0>}]
 dyn_foot_tail_r_MSTR [{shapeType=Capsule, length=1, radius=1}]
}
constraints [
 {bone1Name=dyn_foot_tail_r_01, bone2Name=dyn_foot_tail_r_MSTR, swing1LimitDegrees=45, swing2LimitDegrees=45, twistLimitDegrees=5}
 {bone1Name=dyn_foot_tail_r_02, bone2Name=dyn_foot_tail_r_01, swing1LimitDegrees=45, swing2LimitDegrees=45, twistLimitDegrees=5}
]
