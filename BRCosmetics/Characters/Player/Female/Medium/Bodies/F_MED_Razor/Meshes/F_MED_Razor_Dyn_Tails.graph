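# Physics asset for the dynamic tails and back flap on the F_MED_Razor female player body cosmetic.
# Capsule bodies form left and right top, middle and bottom tail chains plus a back flap chain, all rooted on one pelvis capsule.
# Tail joints use limit angles with tight twist and asymmetric swing, narrower on the top tails.
# Back flap joints lock twist and Swing1, leaving Swing2 as the only unlocked angular axis.
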PhysicsAsset /BRCosmetics/Characters/Player/Female/Medium/Bodies/F_MED_Razor/Meshes/F_MED_Razor_Dyn_Tails
bodies {
 dyn_back_flap_01 [{shapeType=Capsule, length=12.5, radius=1.3, shapeName=dyn_back_flap_01_capsule, center=<5.2296 0 -0.4434>, rotation=<rot 0 0 90>}]
 dyn_back_flap_02 [{shapeType=Capsule, length=12, radius=2.3, shapeName=dyn_back_flap_02_capsule, center=<5.3541 0 0>, rotation=<rot 0 0 90>}]
 dyn_back_flap_03 [{shapeType=Capsule, length=12, radius=3.2, shapeName=dyn_back_flap_03_capsule, center=<3.9553 0 0>, rotation=<rot 0 0 90>}]
 dyn_tails_bott_l_01 [{shapeType=Capsule, length=24, radius=1, center=<12 0 0.0001>, rotation=<rot 90 0 0>}]
 dyn_tails_bott_l_02 [{shapeType=Capsule, length=16.0811, radius=2.4189, center=<9 0.0004 0>, rotation=<rot 90 0 0>}]
 dyn_tails_bott_l_03 [{shapeType=Capsule, length=2.7096, radius=2.5404, center=<2.5 0 0>, rotation=<rot 90 0 0>}]
 dyn_tails_bott_l_04 [{shapeType=Capsule, length=2.192, radius=3.558, center=<2.7 0 0>, rotation=<rot 90 0 0>}]
 dyn_tails_bott_r_01 [{shapeType=Capsule, length=24, radius=1, center=<-12 0 0>, rotation=<rot 90 0 0>}]
 dyn_tails_bott_r_02 [{shapeType=Capsule, length=16.0811, radius=2.4189, center=<-9 0 0>, rotation=<rot 90 0 0>}]
 dyn_tails_bott_r_03 [{shapeType=Capsule, length=2.2087, radius=3.0413, center=<-2.5 0 0>, rotation=<rot 90 0 0>}]
 dyn_tails_bott_r_04 [{shapeType=Capsule, length=1.7735, radius=3.9765, center=<-2.7 0 0>, rotation=<rot 90 0 0>}]
 dyn_tails_mid_l_01 [{shapeType=Capsule, length=15.0476, radius=1, center=<8 0.0001 0.0001>, rotation=<rot 90 0 0>}]
 dyn_tails_mid_l_02 [{shapeType=Capsule, length=15.2884, radius=1.9026, center=<8 0.0003 0>, rotation=<rot 90 0 0>}]
 dyn_tails_mid_l_03 [{shapeType=Capsule, length=9.4503, radius=1.9026, center=<6 0.0002 0>, rotation=<rot 90 0 0>}]
 dyn_tails_mid_l_04 [{shapeType=Capsule, length=11.3689, radius=1, center=<6 -0.0002 0>, rotation=<rot 90 0 0>}]
 dyn_tails_mid_l_05 [{shapeType=Capsule, length=6.9088, radius=1, center=<3 -0.0001 0>, rotation=<rot 90 0 0>}]
 dyn_tails_mid_r_01 [{shapeType=Capsule, length=15.0476, radius=1, center=<-8 0.0001 0.0001>, rotation=<rot 90 0 0>}]
 dyn_tails_mid_r_02 [{shapeType=Capsule, length=15.2884, radius=1.9026, center=<-8 0.0003 0>, rotation=<rot 90 0 0>}]
 dyn_tails_mid_r_03 [{shapeType=Capsule, length=9.4503, radius=1.9026, center=<-6 0.0002 0>, rotation=<rot 90 0 0>}]
 dyn_tails_mid_r_04 [{shapeType=Capsule, length=11.3689, radius=1, center=<-6 -0.0002 0>, rotation=<rot 90 0 0>}]
 dyn_tails_mid_r_05 [{shapeType=Capsule, length=6.9088, radius=1, center=<-3 -0.0001 0>, rotation=<rot 90 0 0>}]
 dyn_tails_top_l_01 [{shapeType=Capsule, length=12.5237, radius=1, center=<7 -0.0001 0.0001>, rotation=<rot 90 0 0>}]
 dyn_tails_top_l_02 [{shapeType=Capsule, length=6.0892, radius=1, center=<4 0.0004 0>, rotation=<rot 90 0 0>}]
 dyn_tails_top_l_03 [{shapeType=Capsule, length=4.25, radius=1, center=<2.5 0 0>, rotation=<rot 90 0 0>}]
 dyn_tails_top_r_01 [{shapeType=Capsule, length=12.5237, radius=1, center=<-7 -0.0002 0.0001>, rotation=<rot 90 0 0>}]
 dyn_tails_top_r_02 [{shapeType=Capsule, length=6.0892, radius=1, center=<-4 0.0004 0>, rotation=<rot 90 0 0>}]
 dyn_tails_top_r_03 [{shapeType=Capsule, length=4.25, radius=1, center=<-2.5 0 0>, rotation=<rot 90 0 0>}]
 pelvis [{shapeType=Capsule, length=9.5, radius=10.5, center=<-4.189 -2.8518 0>}]
}
constraints [
 {bone1Name=dyn_tails_bott_l_02, bone2Name=dyn_tails_bott_l_01, swing1LimitDegrees=60, swing2LimitDegrees=20, twistLimitDegrees=5}
 {bone1Name=dyn_tails_bott_l_03, bone2Name=dyn_tails_bott_l_02, swing1LimitDegrees=60, swing2LimitDegrees=20, twistLimitDegrees=5}
 {bone1Name=dyn_tails_bott_l_04, bone2Name=dyn_tails_bott_l_03, swing1LimitDegrees=60, swing2LimitDegrees=20, twistLimitDegrees=5}
 {bone1Name=dyn_tails_top_l_02, bone2Name=dyn_tails_top_l_01, swing1LimitDegrees=40, swing2LimitDegrees=20, twistLimitDegrees=5}
 {bone1Name=dyn_tails_top_l_03, bone2Name=dyn_tails_top_l_02, swing1LimitDegrees=40, swing2LimitDegrees=20, twistLimitDegrees=5}
 {bone1Name=dyn_tails_mid_l_02, bone2Name=dyn_tails_mid_l_01, swing1LimitDegrees=60, swing2LimitDegrees=20, twistLimitDegrees=5}
 {bone1Name=dyn_tails_mid_l_03, bone2Name=dyn_tails_mid_l_02, swing1LimitDegrees=60, swing2LimitDegrees=20, twistLimitDegrees=5}
 {bone1Name=dyn_tails_mid_l_04, bone2Name=dyn_tails_mid_l_03, swing1LimitDegrees=60, swing2LimitDegrees=20, twistLimitDegrees=5}
 {bone1Name=dyn_tails_mid_l_05, bone2Name=dyn_tails_mid_l_04, swing1LimitDegrees=60, swing2LimitDegrees=20, twistLimitDegrees=5}
 {bone1Name=dyn_tails_bott_r_02, bone2Name=dyn_tails_bott_r_01, swing1LimitDegrees=60, swing2LimitDegrees=20, twistLimitDegrees=5}
 {bone1Name=dyn_tails_bott_r_03, bone2Name=dyn_tails_bott_r_02, swing1LimitDegrees=60, swing2LimitDegrees=20, twistLimitDegrees=5}
 {bone1Name=dyn_tails_bott_r_04, bone2Name=dyn_tails_bott_r_03, swing1LimitDegrees=60, swing2LimitDegrees=20, twistLimitDegrees=5}
 {bone1Name=dyn_tails_top_r_02, bone2Name=dyn_tails_top_r_01, swing1LimitDegrees=40, swing2LimitDegrees=20, twistLimitDegrees=5}
 {bone1Name=dyn_tails_top_r_03, bone2Name=dyn_tails_top_r_02, swing1LimitDegrees=40, swing2LimitDegrees=20, twistLimitDegrees=5}
 {bone1Name=dyn_tails_mid_r_02, bone2Name=dyn_tails_mid_r_01, swing1LimitDegrees=60, swing2LimitDegrees=20, twistLimitDegrees=5}
 {bone1Name=dyn_tails_mid_r_03, bone2Name=dyn_tails_mid_r_02, swing1LimitDegrees=60, swing2LimitDegrees=20, twistLimitDegrees=5}
 {bone1Name=dyn_tails_mid_r_04, bone2Name=dyn_tails_mid_r_03, swing1LimitDegrees=60, swing2LimitDegrees=20, twistLimitDegrees=5}
 {bone1Name=dyn_tails_mid_r_05, bone2Name=dyn_tails_mid_r_04, swing1LimitDegrees=60, swing2LimitDegrees=20, twistLimitDegrees=5}
 {bone1Name=dyn_tails_bott_l_01, bone2Name=pelvis, swing1LimitDegrees=60, swing2LimitDegrees=20, twistLimitDegrees=5}
 {bone1Name=dyn_tails_top_l_01, bone2Name=pelvis, swing1LimitDegrees=40, swing2LimitDegrees=20, twistLimitDegrees=5}
 {bone1Name=dyn_tails_mid_l_01, bone2Name=pelvis, swing1LimitDegrees=60, swing2LimitDegrees=20, twistLimitDegrees=5}
 {bone1Name=dyn_tails_bott_r_01, bone2Name=pelvis, swing1LimitDegrees=60, swing2LimitDegrees=20, twistLimitDegrees=5}
 {bone1Name=dyn_tails_top_r_01, bone2Name=pelvis, swing1LimitDegrees=40, swing2LimitDegrees=20, twistLimitDegrees=5}
 {bone1Name=dyn_tails_mid_r_01, bone2Name=pelvis, swing1LimitDegrees=60, swing2LimitDegrees=20, twistLimitDegrees=5}
 {
  bone1Name=dyn_back_flap_01
  bone2Name=pelvis
  swing1LimitDegrees=45
  swing1Motion=Locked
  swing2LimitDegrees=15
  twistLimitDegrees=45
  twistMotion=Locked
 }
 {
  bone1Name=dyn_back_flap_02
  bone2Name=dyn_back_flap_01
  swing1LimitDegrees=45
  swing1Motion=Locked
  swing2LimitDegrees=15
  twistLimitDegrees=45
  twistMotion=Locked
 }
 {
  bone1Name=dyn_back_flap_03
  bone2Name=dyn_back_flap_02
  swing1LimitDegrees=45
  swing1Motion=Locked
  swing2LimitDegrees=15
  twistLimitDegrees=45
  twistMotion=Locked
 }
]
