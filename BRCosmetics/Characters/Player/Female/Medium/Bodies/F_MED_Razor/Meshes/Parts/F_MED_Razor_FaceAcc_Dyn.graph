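# Physics asset for the face accessory part of the female medium Razor player body, defining secondary motion for hair, bangs and a tassel.
# Two- and three-bone chains for five bangs, back, front and top hair, and the tassel each constrain to the head, which carries two larger capsules.
# Bodies are capsules except the box-shaped bang tips.
# Joints have tight twist limits and per-chain swing limits, narrower on the tassel root and center back hair.
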PhysicsAsset /BRCosmetics/Characters/Player/Female/Medium/Bodies/F_MED_Razor/Meshes/Parts/F_MED_Razor_FaceAcc_Dyn
bodies {
 dyn_bangs_01 [{shapeType=Capsule, length=2.5, radius=0.3, shapeName=dyn_bangs_01_capsule, center=<0.8576 0 0.0001>, rotation=<rot 90 0 0>}]
 dyn_bangs_02 [{shapeType=Box, extentX=2.5, extentY=2.5, extentZ=0.3, shapeName=dyn_bangs_02_box, center=<0.8301 0 0>}]
 dyn_bangs_l_01 [{shapeType=Capsule, length=2.5, radius=0.3, shapeName=dyn_bangs_l_01_capsule, center=<0.8576 0 0.0001>, rotation=<rot 90 0 0>}]
 dyn_bangs_l_02 [{shapeType=Box, extentX=2.5, extentY=2.5, extentZ=0.3, shapeName=dyn_bangs_l_02_box, center=<0.8301 0 0>}]
 dyn_bangs_mid_l_01 [{shapeType=Capsule, length=2.5, radius=0.3, shapeName=dyn_bangs_mid_l_01_capsule, center=<0.8576 0 0.0001>, rotation=<rot 90 0 0>}]
 dyn_bangs_mid_l_02 [{shapeType=Box, extentX=2.5, extentY=2.5, extentZ=0.3, shapeName=dyn_bangs_mid_l_02_box, center=<0.8301 0 0>}]
 dyn_bangs_mid_r_01 [{shapeType=Capsule, length=2.5, radius=0.3, shapeName=dyn_bangs_mid_r_01_capsule, center=<0.8576 0 0.0001>, rotation=<rot 90 0 0>}]
 dyn_bangs_mid_r_02 [{shapeType=Box, extentX=2.5, extentY=2.5, extentZ=0.3, shapeName=dyn_bangs_mid_r_02_box, center=<0.8301 0 0>}]
 dyn_bangs_r_01 [{shapeType=Capsule, length=2.5, radius=0.3, shapeName=dyn_bangs_r_01_capsule, center=<0.8576 0 0.0001>, rotation=<rot 90 0 0>}]
 dyn_bangs_r_02 [{shapeType=Box, extentX=2.5, extentY=2.5, extentZ=0.3, shapeName=dyn_bangs_r_02_box, center=<0.8301 0 0>}]
 dyn_hair_bk_01 [{shapeType=Capsule, length=7.5, radius=1, shapeName=dyn_hair_bk_01_capsule, center=<2.7689 0 0.0002>, rotation=<rot 0 0 90>}]
 dyn_hair_bk_02 [{shapeType=Capsule, length=7.5, radius=0.8, shapeName=dyn_hair_bk_02_capsule, center=<2.7689 0 0.0002>, rotation=<rot 0 0 90>}]
 dyn_hair_bk_03 [{shapeType=Capsule, length=7.5, radius=0.8, shapeName=dyn_hair_bk_03_capsule, center=<2.2382 0 0.0002>, rotation=<rot 0 0 90>}]
 dyn_hair_bk_l_01 [{shapeType=Capsule, length=7.5, radius=0.8, shapeName=dyn_hair_bk_l_01_capsule, center=<2.7689 0 0.0002>, rotation=<rot 0 0 90>}]
 dyn_hair_bk_l_02 [{shapeType=Capsule, length=7.5, radius=0.8, shapeName=dyn_hair_bk_l_02_capsule, center=<2.7689 0 0.0002>, rotation=<rot 0 0 90>}]
 dyn_hair_bk_l_03 [{shapeType=Capsule, length=7.5, radius=0.8, shapeName=dyn_hair_bk_l_03_capsule, center=<2.2382 0 0.0002>, rotation=<rot 0 0 90>}]
 dyn_hair_bk_r_01 [{shapeType=Capsule, length=7.5, radius=0.8, shapeName=dyn_hair_bk_r_01_capsule, center=<2.7689 0 0.0002>, rotation=<rot 0 0 90>}]
 dyn_hair_bk_r_02 [{shapeType=Capsule, length=7.5, radius=0.8, shapeName=dyn_hair_bk_r_02_capsule, center=<2.7689 0 0.0002>, rotation=<rot 0 0 90>}]
 dyn_hair_bk_r_03 [{shapeType=Capsule, length=7.5, radius=0.8, shapeName=dyn_hair_bk_r_03_capsule, center=<2.2382 0 0.0002>, rotation=<rot 0 0 90>}]
 dyn_hair_fr_l_01 [{shapeType=Capsule, length=7.5, radius=0.8, shapeName=dyn_hair_fr_l_01_capsule, center=<2.7689 0 0.0002>, rotation=<rot 0 0 90>}]
 dyn_hair_fr_l_02 [{shapeType=Capsule, length=7.5, radius=0.8, shapeName=dyn_hair_fr_l_02_capsule, center=<1.5237 0 0.0002>, rotation=<rot 0 0 90>}]
 dyn_hair_fr_l_03 [{shapeType=Capsule, length=7.5, radius=0.8, shapeName=dyn_hair_fr_l_03_capsule, center=<1.5237 0 0.0002>, rotation=<rot 0 0 90>}]
 dyn_hair_fr_r_01 [{shapeType=Capsule, length=7.5, radius=0.8, shapeName=dyn_hair_fr_r_01_capsule, center=<2.7689 0 0.0002>, rotation=<rot 0 0 90>}]
 dyn_hair_fr_r_02 [{shapeType=Capsule, length=7.5, radius=0.8, shapeName=dyn_hair_fr_r_02_capsule, center=<2.7689 0 0.0002>, rotation=<rot 0 0 90>}]
 dyn_hair_fr_r_03 [{shapeType=Capsule, length=7.5, radius=0.8, shapeName=dyn_hair_fr_r_03_capsule, center=<2.2382 0 0.0002>, rotation=<rot 0 0 90>}]
 dyn_hair_top_01 [{shapeType=Capsule, length=3, radius=1, center=<1.7956 0 -1.2421>, rotation=<rot 55.0001 0 0>}]
 dyn_hair_top_02 [{shapeType=Capsule, length=3, radius=1, center=<1.7751 0 0.2939>, rotation=<rot 90 0 0>}]
 dyn_hair_top_03 [{shapeType=Capsule, length=3, radius=1, center=<1.8604 0 0>, rotation=<rot 90 0 0>}]
 dyn_tassel_01 [{shapeType=Capsule, length=2.5, radius=0.5, shapeName=dyn_tassel_01_capsule, center=<2 0 0.0001>, rotation=<rot 90 0 0>}]
 dyn_tassel_02 [{shapeType=Capsule, length=2.0024, radius=0.9976, shapeName=dyn_tassel_02_capsule, center=<2 0 0.0001>, rotation=<rot 90 0 0>}]
 head [
  {shapeType=Capsule, length=5.5635, radius=7.7273, center=<2.668 0 -0.4806>, rotation=<rot 90 162.646 172.6461>}
  {shapeType=Capsule, length=5.1759, radius=3.4289, shapeName=head_capsule, center=<4.9503 5.8334 0>}
 ]
}
constraints [
 {bone1Name=dyn_hair_bk_02, bone2Name=dyn_hair_bk_01, swing1LimitDegrees=10, swing2LimitDegrees=30, twistLimitDegrees=5}
 {bone1Name=dyn_hair_bk_03, bone2Name=dyn_hair_bk_02, swing1LimitDegrees=10, swing2LimitDegrees=30, twistLimitDegrees=5}
 {bone1Name=dyn_hair_bk_l_02, bone2Name=dyn_hair_bk_l_01, swing1LimitDegrees=30, swing2LimitDegrees=20, twistLimitDegrees=5}
 {bone1Name=dyn_hair_bk_l_03, bone2Name=dyn_hair_bk_l_02, swing1LimitDegrees=30, swing2LimitDegrees=20, twistLimitDegrees=5}
 {bone1Name=dyn_hair_fr_l_02, bone2Name=dyn_hair_fr_l_01, swing1LimitDegrees=45, swing2LimitDegrees=45, twistLimitDegrees=5}
 {bone1Name=dyn_hair_fr_l_03, bone2Name=dyn_hair_fr_l_02, swing1LimitDegrees=45, swing2LimitDegrees=45, twistLimitDegrees=5}
 {bone1Name=dyn_bangs_mid_l_02, bone2Name=dyn_bangs_mid_l_01, swing1LimitDegrees=45, swing2LimitDegrees=10, twistLimitDegrees=5}
 {bone1Name=dyn_bangs_02, bone2Name=dyn_bangs_01, swing1LimitDegrees=45, swing2LimitDegrees=10, twistLimitDegrees=5}
 {bone1Name=dyn_bangs_mid_r_02, bone2Name=dyn_bangs_mid_r_01, swing1LimitDegrees=45, swing2LimitDegrees=10, twistLimitDegrees=5}
 {bone1Name=dyn_bangs_r_02, bone2Name=dyn_bangs_r_01, swing1LimitDegrees=45, swing2LimitDegrees=10, twistLimitDegrees=5}
 {bone1Name=dyn_bangs_l_02, bone2Name=dyn_bangs_l_01, swing1LimitDegrees=45, swing2LimitDegrees=10, twistLimitDegrees=5}
 {bone1Name=dyn_hair_bk_r_02, bone2Name=dyn_hair_bk_r_01, swing1LimitDegrees=30, swing2LimitDegrees=20, twistLimitDegrees=5}
 {bone1Name=dyn_hair_bk_r_03, bone2Name=dyn_hair_bk_r_02, swing1LimitDegrees=30, swing2LimitDegrees=20, twistLimitDegrees=5}
 {bone1Name=dyn_hair_fr_r_02, bone2Name=dyn_hair_fr_r_01, swing1LimitDegrees=45, swing2LimitDegrees=45, twistLimitDegrees=5}
 {bone1Name=dyn_hair_fr_r_03, bone2Name=dyn_hair_fr_r_02, swing1LimitDegrees=45, swing2LimitDegrees=45, twistLimitDegrees=5}
 {bone1Name=dyn_hair_top_02, bone2Name=dyn_hair_top_01, swing1LimitDegrees=45, swing2LimitDegrees=25, twistLimitDegrees=2}
 {bone1Name=dyn_hair_top_03, bone2Name=dyn_hair_top_02, swing1LimitDegrees=30, swing2LimitDegrees=15, twistLimitDegrees=2}
 {bone1Name=dyn_hair_bk_01, bone2Name=head, swing1LimitDegrees=10, swing2LimitDegrees=30, twistLimitDegrees=5}
 {bone1Name=dyn_hair_bk_l_01, bone2Name=head, swing1LimitDegrees=45, swing2LimitDegrees=45, twistLimitDegrees=5}
 {bone1Name=dyn_hair_fr_l_01, bone2Name=head, swing1LimitDegrees=45, swing2LimitDegrees=45, twistLimitDegrees=5}
 {bone1Name=dyn_bangs_mid_l_01, bone2Name=head, swing1LimitDegrees=45, swing2LimitDegrees=45, twistLimitDegrees=5}
 {bone1Name=dyn_bangs_01, bone2Name=head, swing1LimitDegrees=45, swing2LimitDegrees=45, twistLimitDegrees=5}
 {bone1Name=dyn_bangs_mid_r_01, bone2Name=head, swing1LimitDegrees=45, swing2LimitDegrees=45, twistLimitDegrees=5}
 {bone1Name=dyn_bangs_r_01, bone2Name=head, swing1LimitDegrees=45, swing2LimitDegrees=45, twistLimitDegrees=5}
 {bone1Name=dyn_bangs_l_01, bone2Name=head, swing1LimitDegrees=45, swing2LimitDegrees=45, twistLimitDegrees=5}
 {bone1Name=dyn_hair_bk_r_01, bone2Name=head, swing1LimitDegrees=45, swing2LimitDegrees=45, twistLimitDegrees=5}
 {bone1Name=dyn_hair_fr_r_01, bone2Name=head, swing1LimitDegrees=45, swing2LimitDegrees=45, twistLimitDegrees=5}
 {bone1Name=dyn_hair_top_01, bone2Name=head, swing1LimitDegrees=45, swing2LimitDegrees=45, twistLimitDegrees=5}
 {bone1Name=dyn_tassel_01, bone2Name=head, swing1LimitDegrees=15, swing2LimitDegrees=15, twistLimitDegrees=1}
 {bone1Name=dyn_tassel_02, bone2Name=dyn_tassel_01, swing1LimitDegrees=25, swing2LimitDegrees=25, twistLimitDegrees=2}
]
